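; Looping gameplay cue for the Blizzard Grenade dome, driving its particle effects and an optional on-screen sound indicator.
; On loop start, particles are scaled by RawMagnitude (sphere radius) over the default radius and given a slight random angle variation.
; If enabled, a repeating timer is set inside the particle loop to show the indicator at the target's location.
; Removal passes the same angle to the first particle component and clears the stored timer handle.
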(event EventOnLoopingStart(EitherVFX-preferthisversion)
  (params (MyTarget :type Actor%20Object%20Reference) (Parameters :type Gameplay%20Cue%20Parameters%20Structure%20%28by%20ref%29) (ParticleComponents :type Array%20of%20FXSystem%20Component%20Object%20References) (AudioComponents :type Array%20of%20Audio%20Component%20Object%20References))
  (bind (_normalizedmagnitude _rawmagnitude) (Ability|GameplayCue|BreakGameplayCueParameters Parameters))
  (Variables|Default|SetMy%20Target MyTarget)
  (Variables|Default|SetDomeFX_AngleVariation (Utilities|RandomVectorinRange "-0.050000,-0.050000,-0.050000" "0.050000,0.050000,0.050000"))
  ;[1 1 1] RawMagnitude passed from BGA is the sphere radius
  ; FX treat a radius 512 as the default scale, so we divide by the default to accurately scale the FX
  (for _array_element ParticleComponents
    (Effects|Components|ParticleSystem|SetFloatParameter _array_element "FX_Scale" (/ _rawmagnitude (Variables|Default|GetDefaultDomeSphereRadius)))
    (Effects|Components|ParticleSystem|SetVectorParameter _array_element "FX_DomeAngle" (Variables|Default|GetDomeFX_AngleVariation))
    (if (ScalableFloat|GetValueasBool (Variables|Default|GetHF_SoundIndicatorEnabled))
      (bind _returnvalue (Utilities|Time|SetTimerbyEvent (EventDispatchers|CreateEvent :!event Custom|ShowIndicator) (ScalableFloat|GetValueatLevel (Variables|Default|GetHF_SoundIndicatorRate)) true true))
      (Variables|Default|SetHandle_SoundIndicator _returnvalue))))

(event Custom|ShowIndicator
  (Fortnite|ShowSoundIndicatorLocal self (Variables|Default|GetMy%20Target) (Transformation|GetActorLocation (Variables|Default|GetMy%20Target)) (ScalableFloat|GetValueatLevel (Variables|Default|GetHF_SoundIndicatorDistance)) "Generic" (Utilities|Array|MakeArray (Fortnite|GetFortPlayerControllerfromActor (Variables|Default|GetMy%20Target))) 0 "(R=1.000000,G=1.000000,B=1.000000,A=0.000000)" "/Game/UI/Foundation/Textures/Icons/Navigation/T_UI_BlizzardGrenade_AVIndicator.T_UI_BlizzardGrenade_AVIndicator"))

(event EventOnRemoval(EitherVFX-preferthisversion)
  (params (MyTarget :type Actor%20Object%20Reference) (Parameters :type Gameplay%20Cue%20Parameters%20Structure%20%28by%20ref%29) (ParticleComponents :type Array%20of%20FXSystem%20Component%20Object%20References) (AudioComponents :type Array%20of%20Audio%20Component%20Object%20References) (BurstCameraShakeInstance :type Camera%20Shake%20Base%20Object%20Reference) (BurstDecalInstance :type Decal%20Actor%20Object%20Reference))
  (|CallParentFunction MyTarget Parameters ParticleComponents AudioComponents BurstCameraShakeInstance BurstDecalInstance)
  (Effects|Components|ParticleSystem|SetVectorParameter (Utilities|Array|Get(acopy) ParticleComponents) "FX_DomeAngle" (Variables|Default|GetDomeFX_AngleVariation))
  (Utilities|Time|ClearandInvalidateTimerbyHandle self (Variables|Default|GetHandle_SoundIndicator)))

(fn ConstructionScript)
